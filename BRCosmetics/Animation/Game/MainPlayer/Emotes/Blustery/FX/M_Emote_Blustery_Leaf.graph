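# Masked, two-sided Default Lit material for leaf particles in the Blustery emote, usable on Niagara sprites and mesh particles.
# The leaf texture is desaturated by a dynamic parameter, brightened with a power curve, tinted by particle color and fed into ConvertFromDiffSpec.
# A Subsurface parameter scales that color, and particle alpha multiplies texture alpha.
# The WPO parameter and vertex normal and tangent nodes have no visible consumers.
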Material /BRCosmetics/Animation/Game/MainPlayer/Emotes/Blustery/FX/M_Emote_Blustery_Leaf
graph {
 properties {
  bAutomaticallySetUsageInEditor=false
  bUsedWithNiagaraMeshParticles=true
  bUsedWithNiagaraSprites=true
  blendMode=BLEND_Masked
  translucencyLightingMode=TLM_VolumetricPerVertexNonDirectional
  twoSided=true
  usedShadingModels=Default Lit
 }
 nodes [
  n1 = ParticleColor()
  n2 = TextureSample(UVs=TextureCoordinate()) {texture=&/T_Leaf_Blustery}
  n3 = DynamicParameter() {defaultValue=<rgba 1 0.05 1 1>, paramNames [Desaturation,Subsurface,WPO,Param4]}
  n4 = Desaturation(None=@n2.RGB, Fraction=@n3.Desaturation) {luminanceFactors=<rgba 1 1 1 1>}
  n5 = Saturate(None=Multiply(A=@n1.RGB, B=Power(Base=@n4) {constExponent=0.25}))
  MaterialFunctionCall(DiffuseColor=@n5) {materialFunction=&ConvertFromDiffSpec}
  Multiply(A=@n5, B=@n3.Subsurface) {constB=0.05}
  Multiply(A=@n1.A, B=@n2.A)
  VertexNormalWS()
  VertexTangentWS()
 ]
}
