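# Niagara effect for the Blustery emote: a wind mesh, kicked-up sprite puffs and swirling autumn-colored leaves around skeletal mesh sockets.
# The wind mesh follows a socket while puffs spawn continuously at another, both tinted by time-of-day color.
# Three near-identical leaf emitters (CPU, CPU with collision, denser GPU with collision) combine socket-tracking vortex, point, attraction, curl-noise and age-curved gravity forces.
# Leaves tumble with physics force and scale in and out.
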NiagaraSystem /BRCosmetics/Animation/Game/MainPlayer/Emotes/Blustery/FX/NS_Emote_Blustery
dependencies {
 usedDataInterfaces [
  &NiagaraDataInterfaceCurve
  &NiagaraDataInterfaceSpriteRendererInfo
  &NiagaraDataInterfaceSkeletalMesh
  &NiagaraDataInterfaceSocketReader
  &NiagaraDataInterfaceColorCurve
 ]
 usedDynamicInputs [
  /Niagara/DynamicInputs/Math/GetMatrixRotation
  /Niagara/DynamicInputs/Math/MakeTransform
  /Niagara/DynamicInputs/Rotation/RandomQuaternion
  &ColorFromCurve
  &FNTimeOfDayScalar
 ]
 usedModules [
  &self:SystemSpawnScript.SystemScriptSource.SystemScriptGraph.NiagaraNodeAssignment_0.SetVariables_#1
  &SystemState
  &self:SystemSpawnScript.SystemScriptSource.SystemScriptGraph.NiagaraNodeAssignment_2.SetVariables_#2
 ]
}
emitters {
 Em_Wind {
  emitterUpdateScript [
   &EmitterState(Inactive Response=::NewEnumerator1, Life Cycle Mode=::NewEnumerator1, UseLoopDelay=false)
   &SpawnBurst_Instantaneous(Use Spawn Probability=false, UseLoopCountLimit=false)
  ]
  particleSpawnScript [
   &InitializeParticle(
    Lifetime=$Emitter.CurrentLoopDuration
    Mesh Scale=<1 1 3.5>
    Mesh Scale Mode=::NewEnumerator4
    UsePositionOffset=false
    Write Material Random=true
   )
   &InitialMeshOrientation(
    Mesh Orientation Mode=::NewEnumerator2
    Orientation Coordinate Space=::Local
    Orientation Vector=<0 0 1>
    Rotation=<0 0 -0.25>
    Rotation Coordinate Space=::NewEnumerator2
    Use Rotation Vector=true
   )
  ]
  particleUpdateScript [
   &SkeletalMeshLocation(
    Bone / Socket Index=2
    MaskSpawnGroup=false
    Mesh Sampling Type=::NewEnumerator5
    Sampled Position Offset Coordinate Space:ENiagaraSkelMeshTransforms=::NewEnumerator0
    Skeletal Mesh=$System.Skeletal Mesh
    Socket Sampling Mode=::NewEnumerator5
   )
   &ParticleState()
   &DynamicMaterialParameters(
    Index 0 Param 1=2
    Index 0 Param 2=3
    Module.Index0Param0Write=true
    Module.Index0Param1Write=true
    Module.Index0Param2Write=true
    Module.Index0Param3Write=true
   )
   &ScaleColor(
    Scale Alpha=&FloatFromCurve(FloatCurve=DataInterface() {propertyValues {Curve {keys [
         {arriveTangent=9.0725, interpMode=RCIM_Cubic, leaveTangent=9.0725, tangentMode=RCTM_User, time=0, value=1}
         {arriveTangent=-2.5249, interpMode=RCIM_Cubic, leaveTangent=-2.5249, tangentMode=RCTM_User, time=0.1408, value=2}
         {arriveTangent=-0.0405, interpMode=RCIM_Cubic, leaveTangent=-0.0405, tangentMode=RCTM_User, time=1, value=1}
        ]}}})
    Scale RGB=&MakeVectorFromLinearColorRGB(Color=&MultiplyLinearColor(Linear Color=<rgba 1 1 1 1>, Scale Factor=&MakeVector4_FromLinearColor(LinearColor=$System.ToD Color)))
    ScaleA=true
    ScaleRGB=true
   )
  ]
  rendererClasses [&NiagaraMeshRendererProperties]
  renderers [{rendererClass=&NiagaraMeshRendererProperties, rendererIndex=0}]
 }
 Em_Wind_Kickup {
  emitterUpdateScript [
   &EmitterState(Inactive Response=::NewEnumerator1, Life Cycle Mode=::NewEnumerator1, UseLoopDelay=false)
   &SpawnRate(SpawnRate=24, Use Spawn Probability=false)
  ]
  particleSpawnScript [
   &SocketLocation(
    SamplePosOffset=false
    Sampled Position Offset=<0 5 0>
    Socket Index=1
    Socket Reader=$System.Socket Reader Interface
    Socket Sampling Mode=::NewEnumerator3
    VelocitySpeedLimit=false
   ) {module=SocketLocation001}
   &InitializeParticle(
    Color=<rgba 1 1 1 0.35>
    Lifetime=&RandomRangeFloat(Minimum=0.5, Maximum=0.65)
    Position=$Output.SocketLocation001.SampledPosition
    Position Mode=::NewEnumerator1
    Sprite Rotation Mode=::NewEnumerator2
    Sprite Size Mode=::NewEnumerator1
    Sprite UV Mode=::NewEnumerator4
    Sprite UV Scale:Vector2f=<-1 1>
    Uniform Sprite Size Max=125
    Uniform Sprite Size Min=75
    UsePositionOffset=false
    Write Material Random=true
   )
   &AddVelocity(Velocity=<0 100 100>)
   &AddVelocity(Module.Use Falloff Distance=false, Override Seed=false, Velocity Mode=::NewEnumerator1, Velocity Speed=100) {module=AddVelocity001}
   &ApplyInitialForces()
   &ShapeLocation(
    Cylinder Radius=25
    Enable Lathe Profile=true
    MaskSpawnGroup=false
    Override Seed=false
    Rotation Mode=::NewEnumerator1
    Shape Primitive=::NewEnumerator1
    Surface Only=false
    UseOwnerScale=true
   )
  ]
  particleUpdateScript [
   &ParticleState()
   &SubUVAnimation(
    Play Rate=0.5
    Playback Scale Curve=DataInterface() {propertyValues {Curve {keys [{tangentMode=RCTM_User, time=0, value=0}, {tangentMode=RCTM_User, time=1, value=1}]}}}
    "Sprite Renderer "=DataInterface() {propertyValues {SpriteRenderer=&self:Em_Wind_Kickup_0.NiagaraSpriteRendererProperties_0}}
    SubUV Animation Mode:ENiagara_SubUVLookupModeV2=::NewEnumerator3
    UseEndFrame=false
    UsePlayRateScale=false
    UseStartFrame=false
   )
   &ScaleSpriteSize(Uniform Curve Sprite Scale=DataInterface() {propertyValues {Curve {keys [<at 0 0.5 RCIM_Cubic>, {arriveTangent=2.913, tangentMode=RCTM_User, time=1, value=2}]}}})
   &Drag(Use Linear Drag=true, Use Rotational Drag=true)
   &SolveForcesAndVelocity(Clamp Velocity=false, Limit Acceleration=false)
   &ScaleColor(
    Scale Alpha=&FloatFromCurve(FloatCurve=DataInterface() {propertyValues {Curve {keys [
         <at 0 0>
         {arriveTangent=0.3333, interpMode=RCIM_Cubic, leaveTangent=0.3333, time=0.3, value=0}
         {interpMode=RCIM_Cubic, tangentMode=RCTM_User, time=0.6, value=0.2}
         {tangentMode=RCTM_User, time=1, value=0}
        ]}}})
    Scale RGB=&MakeVectorFromLinearColorRGB(Color=&MultiplyLinearColor(Linear Color=<rgba 1 1 1 1>, Scale Factor=&MakeVector4_FromLinearColor(LinearColor=$System.ToD Color)))
    ScaleA=true
    ScaleRGB=true
   )
  ]
  rendererClasses [&NiagaraSpriteRendererProperties]
  renderers [{rendererClass=&NiagaraSpriteRendererProperties}]
 }
 Rolling_Leaves_CPU {
  emitterUpdateScript [
   &EmitterState()
   &SpawnRate(Spawn Probability=0.25, SpawnRate=128, Use Spawn Probability=false) {module=SpawnRate001}
  ]
  particleSpawnScript [
   &SocketLocation(
    Interpolate Socket Sampling=true
    Orientation Sampling=::NewEnumerator1
    Position Sampling=::NewEnumerator1
    SamplePosOffset=true
    Sampled Orientation Coordinate Space:ENiagaraCoordinateSpace=::Local
    Sampled Position Offset=<100 100 -2>
    Socket Reader=$System.Socket Reader Interface
    Socket Sampling Mode=::NewEnumerator3
    VelocitySpeedLimit=false
   )
   &SocketLocation(
    Orientation Sampling=::NewEnumerator1
    Position Sampling=::NewEnumerator1
    SamplePosOffset=true
    Sampled Position Offset=<0 100 0>
    Sampled Position Offset Coordinate Space:ENiagaraSkelMeshTransforms=::NewEnumerator2
    Socket Index=2
    Socket Reader=$System.Socket Reader Interface
    Socket Sampling Mode=::NewEnumerator3
    VelocitySpeedLimit=false
   ) {module=SocketLocation001}
   &InitializeParticle(
    Color Maximum=<rgba 0.7906 0.3939 0 1>
    Color Minimum=<rgba 0.6573 0.1236 0 1>
    Color Mode=::NewEnumerator2
    Lifetime Max=2.5
    Lifetime Min=1.5
    Lifetime Mode=::NewEnumerator1
    Mesh Scale Mode=::NewEnumerator3
    Mesh Uniform Scale=1.5
    UsePositionOffset=false
    Write Material Random=true
   )
   &ShapeLocation(
    Disc Coverage=1
    MaskSpawnGroup=false
    Offset=<0 0 -40>
    Offset Mode=::NewEnumerator0
    Override Seed=false
    Ring Radius=150
    Rotation Mode=::NewEnumerator2
    Shape Origin=$Output.SocketLocation.SampledPosition
    Shape Primitive=::NewEnumerator4
    U Distribution=0.5
    UseOwnerScale=false
    Yaw / Pitch / Roll=<0.5 0 0>
   ) {module=ShapeLocation001}
   &ShapeLocation(
    Apply To Particle Position=::NewEnumerator1
    Cylinder Radius=15
    Enable Lathe Profile=true
    Lathe Profile=DataInterface() {propertyValues {Curve {keys [<at 0 0.451 RCIM_Cubic>, {arriveTangent=0.049, interpMode=RCIM_Cubic, leaveTangent=0.049, time=0.5, value=1}, <at 1 0.5 RCIM_Cubic>]}}}
    MaskSpawnGroup=false
    Offset Mode=::NewEnumerator0
    Override Seed=false
    Shape Origin=$Output.SocketLocation001.SampledPosition
    Shape Primitive=::NewEnumerator1
    Surface Only=false
    Surface Only Band Thickness=1
    UseOwnerScale=true
   )
   &InitialMeshOrientation(
    Mesh Orientation Mode=::NewEnumerator2
    Rotation=&MakeVector(X=0, Y=0, Z=&RandomRangeFloat(Minimum=-1, Maximum=1))
    Use Rotation Vector=true
   )
  ]
  particleUpdateScript [
   &SocketLocation(
    Position Sampling=::NewEnumerator1
    SamplePosOffset=true
    Sampled Position Offset=<100 100 0>
    Socket Reader=$System.Socket Reader Interface
    Socket Sampling Mode=::NewEnumerator3
    VelocitySpeedLimit=false
   ) {module=SocketLocation003}
   &SocketLocation(
    Orientation Sampling=::NewEnumerator1
    Position Sampling=::NewEnumerator1
    SamplePosOffset=true
    Sampled Position Offset=<0 100 0>
    Sampled Position Offset Coordinate Space:ENiagaraSkelMeshTransforms=::NewEnumerator2
    Socket Index=2
    Socket Reader=$System.Socket Reader Interface
    Socket Sampling Mode=::NewEnumerator3
    VelocitySpeedLimit=false
   ) {module=SocketLocation002}
   &ParticleState()
   &PointForce(
    Force Falloff Distance=75
    Force Origin:NiagaraPosition=$Output.SocketLocation002.SampledPosition
    Force Strength=350
    Linear Force Vector:Vector3f=&TransformVector(Vector=<0 3 3>, Destination Space=::Simulation)
    Module.Use Falloff=true
    Module.UseLinearForce=true
   )
   &VortexForce(
    Origin Pull Amount:NiagaraFloat=&RandomRangeFloat(Minimum=150, Maximum=5000)
    Use Influence Falloff=true
    UseOriginOffset=true
    Vortex Axis=<1 0 0>
    Vortex Axis Coordinate Space:ENiagaraCoordinateSpace=::Local
    Vortex Force Amount=750
    Vortex Origin:NiagaraPosition=$Output.SocketLocation002.SampledPosition
    Vortex Origin Offset:Vector3f=<0 50 7.5>
    Vortex Origin Offset Coordinate Space:ENiagaraCoordinateSpace=::Local
   )
   &CurlNoiseForce(Mask Curl Noise=true, Noise Quality / Cost=::NewEnumerator0, Noise Strength=500, Pan Noise=true, Pan Noise Field=<0 0 150>)
   &Drag(Drag=0.1, Use Linear Drag=true, Use Rotational Drag=true)
   &PointAttractionForce(
    Attraction Radius=75
    Attraction Strength=-50
    Attractor Position:NiagaraPosition=$Output.SocketLocation003.SampledPosition
    Attractor Position Offset:Vector3f=<0 75 75>
    Kill Within Radius=false
    Module.Use Falloff=true
    UseAttractorOffset=true
   )
   &GravityForce(Gravity=&MakeVector(
     X=0
     Y=0
     Z=&FloatFromCurve(
      FloatCurve=DataInterface() {propertyValues {Curve {keys [
          {interpMode=RCIM_Cubic, tangentMode=RCTM_User, time=0, value=0}
          {arriveTangent=-4.8509, interpMode=RCIM_Cubic, leaveTangent=-4.8509, tangentMode=RCTM_User, time=0.5, value=-1}
         ]}}}
      Scale Curve=256
     )
    )) {module=GravityForce001}
   &MeshRotationForce(Rotation=&Multiply_VectorByFloat(Vector=$Transient.PhysicsForce, Float=1))
   &SolveForcesAndVelocity(
    Clamp Velocity=true
    Limit Acceleration=false
    Speed Limit:NiagaraFloat=&RandomRangeFloat(Minimum=200, Maximum=300)
   )
   &ScaleMeshSize(Scale Factor=&VectorFromFloat(Value=&FloatFromCurve(FloatCurve=DataInterface() {propertyValues {Curve {keys [
          {arriveTangent=0.0115, interpMode=RCIM_Cubic, leaveTangent=0.0115, tangentMode=RCTM_User, time=0, value=0}
          <at 0.25 1>
          {arriveTangent=-1.3333, interpMode=RCIM_Cubic, leaveTangent=-1.3333, time=0.85, value=1}
          {arriveTangent=-17.958, interpMode=RCIM_Cubic, leaveTangent=-17.958, tangentMode=RCTM_User, time=1, value=0}
         ]}}})))
   &ScaleColor(
    Scale Alpha=&FloatFromCurve(FloatCurve=DataInterface() {propertyValues {Curve {keys [
         {arriveTangent=18.1065, interpMode=RCIM_Cubic, leaveTangent=18.1065, tangentMode=RCTM_User, time=0.1, value=0}
         {arriveTangent=-0.1196, leaveTangent=-0.1196, time=0.2, value=1}
         {arriveTangent=0.1727, interpMode=RCIM_Cubic, leaveTangent=0.1727, tangentMode=RCTM_User, time=0.8, value=1}
         <at 1 0 RCIM_Cubic>
        ]}}})
    ScaleA=true
    ScaleRGB=false
   ) {enabled=false}
   &DynamicMaterialParameters(
    Index 0 Param 2=&RandomRangeFloat(Minimum=0.1, Maximum=0.3)
    Module.Index0Param0Write=true
    Module.Index0Param1Write=true
    Module.Index0Param2Write=false
    Module.Index0Param3Write=false
   )
   /SetVariables(Particles.Scale=&Multiply_Vector(
     A=$Particles.Initial.Scale
     B=&VectorFromFloat(Value=&FloatFromCurve(FloatCurve=DataInterface() {propertyValues {Curve {keys [
           <at 0.2 1>
           {arriveTangent=-0.0405, interpMode=RCIM_Cubic, leaveTangent=-0.0405, tangentMode=RCTM_User, time=0.85, value=1}
           {arriveTangent=-8.0089, interpMode=RCIM_Cubic, leaveTangent=-8.0089, tangentMode=RCTM_User, time=1, value=0}
          ]}}}))
    )) {isSetParameters=true, module=SetVariables_#3}
  ]
  rendererClasses [&NiagaraMeshRendererProperties]
  renderers [{rendererClass=&NiagaraMeshRendererProperties, rendererIndex=0}]
 }
 Rolling_Leaves_CPU001 {
  emitterUpdateScript [
   &EmitterState()
   &SpawnRate(Spawn Probability=0.25, SpawnRate=128, Use Spawn Probability=false) {module=SpawnRate001}
  ]
  particleSpawnScript [
   &SocketLocation(
    Interpolate Socket Sampling=true
    Orientation Sampling=::NewEnumerator1
    Position Sampling=::NewEnumerator1
    SamplePosOffset=true
    Sampled Orientation Coordinate Space:ENiagaraCoordinateSpace=::Local
    Sampled Position Offset=<100 100 -2>
    Socket Reader=$System.Socket Reader Interface
    Socket Sampling Mode=::NewEnumerator3
    VelocitySpeedLimit=false
   )
   &SocketLocation(
    Orientation Sampling=::NewEnumerator1
    Position Sampling=::NewEnumerator1
    SamplePosOffset=true
    Sampled Position Offset=<0 100 0>
    Sampled Position Offset Coordinate Space:ENiagaraSkelMeshTransforms=::NewEnumerator2
    Socket Index=2
    Socket Reader=$System.Socket Reader Interface
    Socket Sampling Mode=::NewEnumerator3
    VelocitySpeedLimit=false
   ) {module=SocketLocation001}
   &InitializeParticle(
    Color Maximum=<rgba 0.7906 0.3939 0 1>
    Color Minimum=<rgba 0.6573 0.1236 0 1>
    Color Mode=::NewEnumerator2
    Lifetime Max=2.5
    Lifetime Min=1.5
    Lifetime Mode=::NewEnumerator1
    Mesh Scale Mode=::NewEnumerator3
    Mesh Uniform Scale=1.5
    UsePositionOffset=false
    Write Material Random=true
   )
   &ShapeLocation(
    Disc Coverage=1
    MaskSpawnGroup=false
    Offset=<0 0 -40>
    Offset Mode=::NewEnumerator0
    Override Seed=false
    Ring Radius=150
    Rotation Mode=::NewEnumerator2
    Shape Origin=$Output.SocketLocation.SampledPosition
    Shape Primitive=::NewEnumerator4
    U Distribution=0.5
    UseOwnerScale=false
    Yaw / Pitch / Roll=<0.5 0 0>
   ) {module=ShapeLocation001}
   &ShapeLocation(
    Apply To Particle Position=::NewEnumerator1
    Cylinder Radius=15
    Enable Lathe Profile=true
    Lathe Profile=DataInterface() {propertyValues {Curve {keys [<at 0 0.451 RCIM_Cubic>, {arriveTangent=0.049, interpMode=RCIM_Cubic, leaveTangent=0.049, time=0.5, value=1}, <at 1 0.5 RCIM_Cubic>]}}}
    MaskSpawnGroup=false
    Offset Mode=::NewEnumerator0
    Override Seed=false
    Shape Origin=$Output.SocketLocation001.SampledPosition
    Shape Primitive=::NewEnumerator1
    Surface Only=false
    Surface Only Band Thickness=1
    UseOwnerScale=true
   )
   &InitialMeshOrientation(
    Mesh Orientation Mode=::NewEnumerator2
    Rotation=&MakeVector(X=0, Y=0, Z=&RandomRangeFloat(Minimum=-1, Maximum=1))
    Use Rotation Vector=true
   )
  ]
  particleUpdateScript [
   &SocketLocation(
    Position Sampling=::NewEnumerator1
    SamplePosOffset=true
    Sampled Position Offset=<100 100 0>
    Socket Reader=$System.Socket Reader Interface
    Socket Sampling Mode=::NewEnumerator3
    VelocitySpeedLimit=false
   ) {module=SocketLocation003}
   &SocketLocation(
    Orientation Sampling=::NewEnumerator1
    Position Sampling=::NewEnumerator1
    SamplePosOffset=true
    Sampled Position Offset=<0 100 0>
    Sampled Position Offset Coordinate Space:ENiagaraSkelMeshTransforms=::NewEnumerator2
    Socket Index=2
    Socket Reader=$System.Socket Reader Interface
    Socket Sampling Mode=::NewEnumerator3
    VelocitySpeedLimit=false
   ) {module=SocketLocation002}
   &ParticleState()
   &PointForce(
    Force Falloff Distance=75
    Force Origin:NiagaraPosition=$Output.SocketLocation002.SampledPosition
    Force Strength=350
    Linear Force Vector:Vector3f=&TransformVector(Vector=<0 3 3>, Destination Space=::Simulation)
    Module.Use Falloff=true
    Module.UseLinearForce=true
   )
   &VortexForce(
    Origin Pull Amount:NiagaraFloat=&RandomRangeFloat(Minimum=150, Maximum=5000)
    Use Influence Falloff=true
    UseOriginOffset=true
    Vortex Axis=<1 0 0>
    Vortex Axis Coordinate Space:ENiagaraCoordinateSpace=::Local
    Vortex Force Amount=750
    Vortex Origin:NiagaraPosition=$Output.SocketLocation002.SampledPosition
    Vortex Origin Offset:Vector3f=<0 50 7.5>
    Vortex Origin Offset Coordinate Space:ENiagaraCoordinateSpace=::Local
   )
   &CurlNoiseForce(Mask Curl Noise=true, Noise Quality / Cost=::NewEnumerator0, Noise Strength=500, Pan Noise=true, Pan Noise Field=<0 0 150>)
   &Drag(Drag=0.1, Use Linear Drag=true, Use Rotational Drag=true)
   &PointAttractionForce(
    Attraction Radius=75
    Attraction Strength=-50
    Attractor Position:NiagaraPosition=$Output.SocketLocation003.SampledPosition
    Attractor Position Offset:Vector3f=<0 75 75>
    Kill Within Radius=false
    Module.Use Falloff=true
    UseAttractorOffset=true
   )
   &GravityForce(Gravity=&MakeVector(
     X=0
     Y=0
     Z=&FloatFromCurve(
      FloatCurve=DataInterface() {propertyValues {Curve {keys [
          {interpMode=RCIM_Cubic, tangentMode=RCTM_User, time=0, value=0}
          {arriveTangent=-4.8509, interpMode=RCIM_Cubic, leaveTangent=-4.8509, tangentMode=RCTM_User, time=0.5, value=-1}
         ]}}}
      Scale Curve=256
     )
    )) {module=GravityForce001}
   &Collision(
    Age Colliding Particles=false
    Control Roll On Collision=true
    KillOnCollision=false
    Module.Control Rest State Via Surface Normal=true
    Module.EnableMaxCollisionCount=false
    Module.Randomize Collision Normal Vector=true
    Radius Calculation Type=::NewEnumerator1
    Radius Scale for Rotation=3
    ScaleRateOfRotation=true
   )
   &MeshRotationForce(Rotation=&Multiply_VectorByFloat(Vector=$Transient.PhysicsForce, Float=1))
   &SolveForcesAndVelocity(
    Clamp Velocity=true
    Limit Acceleration=false
    Speed Limit:NiagaraFloat=&RandomRangeFloat(Minimum=200, Maximum=300)
   )
   &ScaleMeshSize(Scale Factor=&VectorFromFloat(Value=&FloatFromCurve(FloatCurve=DataInterface() {propertyValues {Curve {keys [
          {arriveTangent=0.0115, interpMode=RCIM_Cubic, leaveTangent=0.0115, tangentMode=RCTM_User, time=0, value=0}
          <at 0.25 1>
          {arriveTangent=-1.3333, interpMode=RCIM_Cubic, leaveTangent=-1.3333, time=0.85, value=1}
          {arriveTangent=-17.958, interpMode=RCIM_Cubic, leaveTangent=-17.958, tangentMode=RCTM_User, time=1, value=0}
         ]}}})))
   &ScaleColor(
    Scale Alpha=&FloatFromCurve(FloatCurve=DataInterface() {propertyValues {Curve {keys [
         {arriveTangent=18.1065, interpMode=RCIM_Cubic, leaveTangent=18.1065, tangentMode=RCTM_User, time=0.1, value=0}
         {arriveTangent=-0.1196, leaveTangent=-0.1196, time=0.2, value=1}
         {arriveTangent=0.1727, interpMode=RCIM_Cubic, leaveTangent=0.1727, tangentMode=RCTM_User, time=0.8, value=1}
         <at 1 0 RCIM_Cubic>
        ]}}})
    ScaleA=true
    ScaleRGB=false
   ) {enabled=false}
   &DynamicMaterialParameters(
    Index 0 Param 2=&RandomRangeFloat(Minimum=0.1, Maximum=0.3)
    Module.Index0Param0Write=true
    Module.Index0Param1Write=true
    Module.Index0Param2Write=false
    Module.Index0Param3Write=false
   )
   /SetVariables(Particles.Scale=&Multiply_Vector(
     A=$Particles.Initial.Scale
     B=&VectorFromFloat(Value=&FloatFromCurve(FloatCurve=DataInterface() {propertyValues {Curve {keys [
           <at 0.2 1>
           {arriveTangent=-0.0405, interpMode=RCIM_Cubic, leaveTangent=-0.0405, tangentMode=RCTM_User, time=0.85, value=1}
           {arriveTangent=-8.0089, interpMode=RCIM_Cubic, leaveTangent=-8.0089, tangentMode=RCTM_User, time=1, value=0}
          ]}}}))
    )) {isSetParameters=true, module=SetVariables_#3}
  ]
  rendererClasses [&NiagaraMeshRendererProperties]
  renderers [{rendererClass=&NiagaraMeshRendererProperties, rendererIndex=0}]
 }
 Rolling_Leaves_GPU001 {
  simTarget=GPUComputeSim
  emitterUpdateScript [
   &EmitterState()
   &SpawnRate(Spawn Probability=0.25, SpawnRate=512, Use Spawn Probability=false) {module=SpawnRate001}
  ]
  particleSpawnScript [
   &SocketLocation(
    Interpolate Socket Sampling=true
    Orientation Sampling=::NewEnumerator1
    Position Sampling=::NewEnumerator1
    SamplePosOffset=true
    Sampled Orientation Coordinate Space:ENiagaraCoordinateSpace=::Local
    Sampled Position Offset=<100 100 0>
    Socket Reader=$System.Socket Reader Interface
    Socket Sampling Mode=::NewEnumerator3
    VelocitySpeedLimit=false
   )
   &SocketLocation(
    Orientation Sampling=::NewEnumerator1
    Position Sampling=::NewEnumerator1
    SamplePosOffset=true
    Sampled Position Offset=<0 100 0>
    Sampled Position Offset Coordinate Space:ENiagaraSkelMeshTransforms=::NewEnumerator2
    Socket Index=2
    Socket Reader=$System.Socket Reader Interface
    Socket Sampling Mode=::NewEnumerator3
    VelocitySpeedLimit=false
   ) {module=SocketLocation001}
   &InitializeParticle(
    Color Maximum=<rgba 0.7906 0.3939 0 1>
    Color Minimum=<rgba 0.6573 0.1236 0 1>
    Color Mode=::NewEnumerator2
    Lifetime Max=2.5
    Lifetime Min=1.5
    Lifetime Mode=::NewEnumerator1
    Mesh Scale Mode=::NewEnumerator3
    Mesh Uniform Scale=1.25
    UsePositionOffset=false
    Write Material Random=true
   )
   &ShapeLocation(
    Disc Coverage=1
    MaskSpawnGroup=false
    Offset=<0 0 -40>
    Offset Mode=::NewEnumerator0
    Override Seed=false
    Ring Radius=150
    Rotation Mode=::NewEnumerator2
    Shape Origin=$Output.SocketLocation.SampledPosition
    Shape Primitive=::NewEnumerator4
    U Distribution=0.5
    UseOwnerScale=false
    Yaw / Pitch / Roll=<0.5 0 0>
   ) {module=ShapeLocation001}
   &ShapeLocation(
    Apply To Particle Position=::NewEnumerator1
    Cylinder Radius=15
    Enable Lathe Profile=true
    Lathe Profile=DataInterface() {propertyValues {Curve {keys [<at 0 0.451 RCIM_Cubic>, {arriveTangent=0.049, interpMode=RCIM_Cubic, leaveTangent=0.049, time=0.5, value=1}, <at 1 0.5 RCIM_Cubic>]}}}
    MaskSpawnGroup=false
    Offset Mode=::NewEnumerator0
    Override Seed=false
    Shape Origin=$Output.SocketLocation001.SampledPosition
    Shape Primitive=::NewEnumerator1
    Surface Only=false
    Surface Only Band Thickness=1
    UseOwnerScale=true
   )
   &InitialMeshOrientation(
    Mesh Orientation Mode=::NewEnumerator2
    Rotation=&MakeVector(X=0, Y=0, Z=&RandomRangeFloat(Minimum=-1, Maximum=1))
    Use Rotation Vector=true
   )
  ]
  particleUpdateScript [
   &SocketLocation(
    Position Sampling=::NewEnumerator1
    SamplePosOffset=true
    Sampled Position Offset=<100 100 0>
    Socket Reader=$System.Socket Reader Interface
    Socket Sampling Mode=::NewEnumerator3
    VelocitySpeedLimit=false
   ) {module=SocketLocation003}
   &SocketLocation(
    Orientation Sampling=::NewEnumerator1
    Position Sampling=::NewEnumerator1
    SamplePosOffset=true
    Sampled Position Offset=<0 100 0>
    Sampled Position Offset Coordinate Space:ENiagaraSkelMeshTransforms=::NewEnumerator2
    Socket Index=2
    Socket Reader=$System.Socket Reader Interface
    Socket Sampling Mode=::NewEnumerator3
    VelocitySpeedLimit=false
   ) {module=SocketLocation002}
   &ParticleState()
   &PointForce(
    Force Falloff Distance=75
    Force Origin:NiagaraPosition=$Output.SocketLocation002.SampledPosition
    Force Strength=350
    Linear Force Vector:Vector3f=&TransformVector(Vector=<0 3 3>, Destination Space=::Simulation)
    Module.Use Falloff=true
    Module.UseLinearForce=true
   )
   &VortexForce(
    Origin Pull Amount:NiagaraFloat=&RandomRangeFloat(Minimum=150, Maximum=5000)
    Use Influence Falloff=true
    UseOriginOffset=true
    Vortex Axis=<1 0 0>
    Vortex Axis Coordinate Space:ENiagaraCoordinateSpace=::Local
    Vortex Force Amount=750
    Vortex Origin:NiagaraPosition=$Output.SocketLocation002.SampledPosition
    Vortex Origin Offset:Vector3f=<0 50 7.5>
    Vortex Origin Offset Coordinate Space:ENiagaraCoordinateSpace=::Local
   )
   &CurlNoiseForce(Mask Curl Noise=true, Noise Quality / Cost=::NewEnumerator0, Noise Strength=500, Pan Noise=true, Pan Noise Field=<0 0 150>)
   &Drag(Drag=0.1, Use Linear Drag=true, Use Rotational Drag=true)
   &PointAttractionForce(
    Attraction Radius=75
    Attraction Strength=-50
    Attractor Position:NiagaraPosition=$Output.SocketLocation003.SampledPosition
    Attractor Position Offset:Vector3f=<0 75 75>
    Kill Within Radius=false
    Module.Use Falloff=true
    UseAttractorOffset=true
   )
   &GravityForce(Gravity=&MakeVector(
     X=0
     Y=0
     Z=&FloatFromCurve(
      FloatCurve=DataInterface() {propertyValues {Curve {keys [
          {interpMode=RCIM_Cubic, tangentMode=RCTM_User, time=0, value=0}
          {arriveTangent=-4.8509, interpMode=RCIM_Cubic, leaveTangent=-4.8509, tangentMode=RCTM_User, time=0.5, value=-1}
         ]}}}
      Scale Curve=256
     )
    )) {module=GravityForce001}
   &Collision(
    Age Colliding Particles=false
    Control Roll On Collision=true
    GPU Collision Type=::NewEnumerator2
    KillOnCollision=false
    Module.Control Rest State Via Surface Normal=true
    Module.EnableMaxCollisionCount=false
    Module.Randomize Collision Normal Vector=true
    Radius Calculation Type=::NewEnumerator1
    Radius Scale for Rotation=3
    ScaleRateOfRotation=true
   )
   &MeshRotationForce(Rotation=&Multiply_VectorByFloat(Vector=$Transient.PhysicsForce, Float=1))
   &SolveForcesAndVelocity(
    Clamp Velocity=true
    Limit Acceleration=false
    Speed Limit:NiagaraFloat=&RandomRangeFloat(Minimum=200, Maximum=300)
   )
   &ScaleMeshSize(Scale Factor=&VectorFromFloat(Value=&FloatFromCurve(FloatCurve=DataInterface() {propertyValues {Curve {keys [
          {arriveTangent=0.06, interpMode=RCIM_Cubic, leaveTangent=0.06, tangentMode=RCTM_User, time=0, value=0}
          <at 0.25 1>
          {arriveTangent=-1.3333, interpMode=RCIM_Cubic, leaveTangent=-1.3333, time=0.85, value=1}
          {arriveTangent=-17.958, interpMode=RCIM_Cubic, leaveTangent=-17.958, tangentMode=RCTM_User, time=1, value=0}
         ]}}})))
   &ScaleColor(
    Scale Alpha=&FloatFromCurve(FloatCurve=DataInterface() {propertyValues {Curve {keys [
         {arriveTangent=18.1065, interpMode=RCIM_Cubic, leaveTangent=18.1065, tangentMode=RCTM_User, time=0.1, value=0}
         {arriveTangent=-0.1196, leaveTangent=-0.1196, time=0.2, value=1}
         {arriveTangent=0.1727, interpMode=RCIM_Cubic, leaveTangent=0.1727, tangentMode=RCTM_User, time=0.8, value=1}
         <at 1 0 RCIM_Cubic>
        ]}}})
    ScaleA=true
    ScaleRGB=false
   ) {enabled=false}
   &DynamicMaterialParameters(
    Index 0 Param 2=&RandomRangeFloat(Minimum=0.1, Maximum=0.3)
    Module.Index0Param0Write=true
    Module.Index0Param1Write=true
    Module.Index0Param2Write=false
    Module.Index0Param3Write=false
   )
   /SetVariables(Particles.Scale=&Multiply_Vector(
     A=$Particles.Initial.Scale
     B=&VectorFromFloat(Value=&FloatFromCurve(FloatCurve=DataInterface() {propertyValues {Curve {keys [
           <at 0.2 1>
           {arriveTangent=-0.0405, interpMode=RCIM_Cubic, leaveTangent=-0.0405, tangentMode=RCTM_User, time=0.85, value=1}
           {arriveTangent=-8.0089, interpMode=RCIM_Cubic, leaveTangent=-8.0089, tangentMode=RCTM_User, time=1, value=0}
          ]}}}))
    )) {isSetParameters=true, module=SetVariables_#3}
  ]
  rendererClasses [&NiagaraMeshRendererProperties]
  renderers [{rendererClass=&NiagaraMeshRendererProperties, rendererIndex=0}]
 }
}
summary {
 systemName=NS_Emote_Blustery
 emitters [
  {emitterName=Em_Wind_Kickup, rendererClasses [&NiagaraSpriteRendererProperties]}
  {emitterName=Rolling_Leaves_GPU001, simTarget=GPUComputeSim, rendererClasses [&NiagaraMeshRendererProperties]}
  {emitterName=Rolling_Leaves_CPU001, rendererClasses [&NiagaraMeshRendererProperties]}
  {emitterName=Em_Wind, rendererClasses [&NiagaraMeshRendererProperties]}
  {emitterName=Rolling_Leaves_CPU, rendererClasses [&NiagaraMeshRendererProperties]}
 ]
 userVariables []
}
